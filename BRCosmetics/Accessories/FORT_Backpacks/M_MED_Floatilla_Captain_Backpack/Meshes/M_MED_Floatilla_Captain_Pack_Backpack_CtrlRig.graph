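# Control Rig for the Floatilla Captain backpack cosmetic that procedurally rotates the dangling baseball bone from its parent's motion.
# It reads dyn_baseball_MSTR's rotation relative to root as ZYX Euler and remaps a narrow, clamped band of X into a much wider Y angle.
# That angle is applied to dyn_baseball as an additive local rotation.
# An AccumulateQuatLerp smoothing node is left unconnected, and the other dyn_ bones are untouched.
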ControlRigBlueprint /BRCosmetics/Accessories/FORT_Backpacks/M_MED_Floatilla_Captain_Backpack/Meshes/M_MED_Floatilla_Captain_Pack_Backpack_CtrlRig
graphs [
 RigVMFunctionLibrary {links [], nodes []}
 RigVMModel {nodes [
   n1 = &RigUnit_BeginExecution()
   n2 = &RigVMFunction_MathQuaternionToEuler(RotationOrder=ZYX, Value=&RigUnit_GetRelativeBoneTransform(Bone=dyn_baseball_MSTR, Space=root).Transform.Rotation)
   /Script/ControlRig.RigUnit_ModifyBoneTransforms(
    BoneToModify=((Bone="dyn_baseball",Transform=(Rotation=(X=0,Y=0,Z=0,W=1),Translation=(X=0,Y=0,Z=0),Scale3D=(X=1,Y=1,Z=1))))
    Mode=AdditiveLocal
    Weight=1
    WeightMaximum=1
    WeightMinimum=0
    BoneToModify.0.Transform.Rotation=&RigVMFunction_MathQuaternionFromEuler(
     Euler=(X=0,Y=0,Z=0)
     Euler.Y=&RigVMFunction_MathFloatRemap(SourceMaximum=-19, SourceMinimum=-30, TargetMaximum=200, TargetMinimum=0, bClamp=true, Value=@n2.Result.X).Result
    ).Result
    ExecutePin=@n1.ExecutePin
   ) {is=struct RigVMUnitNode}
   /Script/RigVM.RigVMFunction_AccumulateQuatLerp(
    Blend=0.35
    TargetValue=(X=0,Y=0,Z=0,W=1)
    bIntegrateDeltaTime=false
    InitialValue=&RigVMFunction_MathQuaternionFromEuler(Euler=(X=0,Y=0,Z=0)).Result
   ) {is=struct RigVMUnitNode, template="AccumulateLerp::Execute(in TargetValue,in InitialValue,in Blend,in bIntegrateDeltaTime,out Result)"}
  ]}
]
hierarchy [
 root {loc=<0 0 0>, rot=<0 0 0 1>}
 pack {parent=root, loc=<0 -0.4484 0>, rot=<0 0 0 1>}
 dyn_teeth_MSTR {parent=pack, loc=<3.0976 -5.6537 -12.7078>, rot=<0 0 0 1>}
 dyn_teeth_01 {parent=dyn_teeth_MSTR, loc=<14.2 0.7026 3.7657>, rot=<0.8172 -0.0589 -0.5602 0.1221>}
 dyn_teeth_02 {parent=dyn_teeth_MSTR, loc=<12.3591 0.1413 2.1491>, rot=<0.8174 -0.038 -0.5583 0.1369>}
 dyn_teeth_03 {parent=dyn_teeth_MSTR, loc=<9.7727 -0.4982 0.1721>, rot=<0.7925 -0.0658 -0.5943 0.1197>}
 dyn_teeth_04 {parent=dyn_teeth_MSTR, loc=<6.3327 -1.2297 -2.1582>, rot=<0.7672 -0.062 -0.6264 0.1235>}
 dyn_teeth_05 {parent=dyn_teeth_MSTR, loc=<2.043 -1.8419 -3.795>, rot=<-0.6534 0.0854 0.7443 -0.1088>}
 dyn_teeth_06 {parent=dyn_teeth_MSTR, loc=<-2.5193 -2.0456 -3.361>, rot=<-0.441 0.1012 0.8879 -0.0827>}
 dyn_teeth_07 {parent=dyn_teeth_MSTR, loc=<-6.37 -1.5342 -0.7427>, rot=<-0.2687 0.113 0.955 -0.0552>}
 dyn_rope_MSTR {parent=pack, loc=<-14.392 -1.8214 8.6025>, rot=<-0.0835 0.0027 0.996 0.0318>}
 dyn_rope_r_01 {parent=dyn_rope_MSTR, loc=<1.2355 -0.8053 0.9408>, rot=<0 0 0 -1>}
 dyn_rope_r_02 {parent=dyn_rope_r_01, loc=<1.3824 -0.0175 -0.1228>, rot=<0 0 0 -1>}
 dyn_rope_r_03 {parent=dyn_rope_r_02, loc=<1.9268 -0.0396 0.015>, rot=<0 0 0 -1>}
 dyn_rope_r_04 {parent=dyn_rope_r_03, loc=<2.1595 -0.0645 -0.006>, rot=<0 0 0 -1>}
 dyn_rope_l_01 {parent=dyn_rope_MSTR, loc=<0.9911 -1.0941 0.017>, rot=<-0.0044 -0.0129 0.0241 -0.9996>}
 dyn_rope_l_02 {parent=dyn_rope_l_01, loc=<2.0447 -0.0761 -0.1088>, rot=<-0.0035 0.0043 0.0184 -0.9998>}
 dyn_rope_l_03 {parent=dyn_rope_l_02, loc=<2.5982 0.0905 -0.0536>, rot=<0.0054 -0.0036 -0.0284 -0.9996>}
 dyn_rope_l_04 {parent=dyn_rope_l_03, loc=<2.6088 -0.0232 -0.0394>, rot=<0 0 0 1>}
 dyn_air_hose_01 {parent=pack, loc=<1.4437 -6.6776 5.6593>, rot=<0.7438 -0.2557 0.4518 0.421>}
 dyn_air_hose_02 {parent=pack, loc=<6.3352 -3.3641 0.85>, rot=<0.0033 0.0334 0.0982 -0.9946>}
 dyn_air_hose_03 {parent=dyn_air_hose_02, loc=<3.7628 0 0>, rot=<-0.2032 -0.4627 0.3948 -0.7673>}
 dyn_belt_buckle {parent=pack, loc=<1.5041 -7.3256 -2.309>, rot=<0.2706 -0.2706 0.6533 -0.6533>}
 dyn_seashell {parent=pack, loc=<4.4579 -10.2799 -4.5394>, rot=<-0.4232 -0.3623 0.4473 -0.6997>}
 dyn_bottlecap {parent=pack, loc=<6.4854 -8.8843 -2.8495>, rot=<-0.4401 -0.3005 0.1676 -0.8294>}
 dyn_top_strap {parent=pack, loc=<-7.815 -1.2649 9.4386>, rot=<0.4634 -0.0034 0.8861 0.0064>}
 dyn_top_strap_01 {parent=dyn_top_strap, loc=<1.3504 -0.0066 -0.0794>, rot=<0 0 0 -1>}
 dyn_baseball_MSTR {parent=pack, loc=<8.704 -4.0546 -4.1814>, rot=<0.1416 -0.5938 0.0494 -0.7905>}
 dyn_baseball {parent=dyn_baseball_MSTR, loc=<6.7437 0 0>, rot=<-0.019 0.5973 -0.0114 -0.8017>}
]
