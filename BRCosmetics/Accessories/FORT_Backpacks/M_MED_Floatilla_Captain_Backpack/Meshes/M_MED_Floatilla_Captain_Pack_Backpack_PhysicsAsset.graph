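# Physics asset for the Floatilla Captain backpack cosmetic, defining collision bodies and joints for its dangling parts.
# The multi-capsule pack body anchors the straps, buckle, baseball, seashell, bottlecap and air hose pieces through angular limits of varying tightness.
# A rope master branches into two three-link chains.
# Seven tightly limited teeth hang from a teeth master whose swing and twist relative to the pack are free.
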PhysicsAsset /BRCosmetics/Accessories/FORT_Backpacks/M_MED_Floatilla_Captain_Backpack/Meshes/M_MED_Floatilla_Captain_Pack_Backpack_PhysicsAsset
bodies {
 dyn_air_hose_01 [{shapeType=Capsule, length=0.7021, radius=4.1768, center=<2.789 0.1169 0.09>, rotation=<rot 90 0 0>}]
 dyn_air_hose_02 [{shapeType=Capsule, length=1.7955, radius=1.9347, center=<1.5055 0.3461 0.0859>, rotation=<rot 0 102.5285 -71.9205>}]
 dyn_air_hose_03 [{shapeType=Capsule, length=3.726, radius=3.4686, center=<3.1257 0.1038 -1.39>, rotation=<rot 0 -96.9215 -50.0863>}]
 dyn_baseball_MSTR [{shapeType=Sphere, radius=2.5516, center=<6.7803 0 0>}]
 dyn_belt_buckle [{shapeType=Box, extentX=4.5902, extentY=4.282, extentZ=11.2111, center=<3.6029 -0.1316 -0.4641>, rotation=<rot 3.3013 -90 5.1327>}]
 dyn_bottlecap [{shapeType=Box, extentX=3.1382, extentY=1.3492, extentZ=3.0936, center=<2.3984 -0.1691 -0.0995>, rotation=<rot 14.899 -95.0129 -88.1949>}]
 dyn_rope_MSTR [{shapeType=Sphere, radius=0.7555, center=<0.4322 -0.0702 0.0684>}]
 dyn_rope_l_01 [{shapeType=Capsule, length=0.8959, radius=0.505, center=<0.6607 -0.067 -0.0282>, rotation=<rot -30.6408 -89.9997 -87.4601>}]
 dyn_rope_l_02 [{shapeType=Capsule, length=1.6773, radius=0.6061, center=<1.1187 -0.0284 -0.077>, rotation=<rot 0 -87.4385 -88.5517>}]
 dyn_rope_l_03 [{shapeType=Capsule, length=1.6868, radius=0.6387, center=<1.7098 0.0617 -0.0032>, rotation=<rot 32.9017 -89.0109 -89.9996>}]
 dyn_rope_r_01 [{shapeType=Capsule, length=0.505, radius=0.505, center=<0.2719 -0.0286 -0.1046>, rotation=<rot 90 -0.0049 -0.0049>}]
 dyn_rope_r_02 [{shapeType=Capsule, length=1.3808, radius=0.505, center=<1.1317 -0.0079 0.0268>, rotation=<rot 90 -0.0297 -0.0299>}]
 dyn_rope_r_03 [{shapeType=Capsule, length=1.4667, radius=0.505, center=<1.7649 0.0238 0.0213>, rotation=<rot 90 -0.0098 -0.0098>}]
 dyn_seashell [{shapeType=Box, extentX=4.4769, extentY=1.3441, extentZ=3.7585, center=<2.8938 0.0951 -0.2383>, rotation=<rot -4.8614 -91.4236 -76.4463>}]
 dyn_teeth_01 [{shapeType=Box, extentX=0.5718, extentY=1.511, extentZ=2.0392, center=<0.5047 -0.079 0.1377>, rotation=<rot 0 88.93 -51.5625>}]
 dyn_teeth_02 [{shapeType=Box, extentX=0.8064, extentY=2.1004, extentZ=2.8763, center=<0.675 -0.0879 0.1703>, rotation=<rot 0 88.8908 -51.2797>}]
 dyn_teeth_03 [{shapeType=Box, extentX=0.8728, extentY=2.6455, extentZ=3.6328, center=<0.8796 -0.1075 0.1363>, rotation=<rot 0 89.4849 -53.2111>}]
 dyn_teeth_04 [{shapeType=Box, extentX=1.0381, extentY=3.5377, extentZ=4.6418, center=<0.9756 -0.056 -0.0514>, rotation=<rot 0 88.6296 -49.4617>}]
 dyn_teeth_05 [{shapeType=Box, extentX=1.1935, extentY=3.7169, extentZ=5.497, center=<1.1429 0.0017 0.106>, rotation=<rot 0 89.1789 -56.4709>}]
 dyn_teeth_06 [{shapeType=Box, extentX=1.3148, extentY=3.9519, extentZ=5.7122, center=<1.2255 -0.0985 0.0977>, rotation=<rot 0 89.7098 -56.6296>}]
 dyn_teeth_07 [{shapeType=Box, extentX=1.3365, extentY=4.3246, extentZ=5.8025, center=<1.3502 -0.0891 0.0771>, rotation=<rot 0 90.2939 -55.0215>}]
 dyn_teeth_MSTR [{shapeType=Box, extentX=6.5904, extentY=9.4515, extentZ=12.9165, center=<1.6715 0 3.3811>}]
 dyn_top_strap [{shapeType=Capsule, length=1.7739, radius=1.5983, center=<0.0585 -0.0662 -0.0576>, rotation=<rot -6.8683 -90 -94.6653>}]
 pack [
  {shapeType=Capsule, length=2.3987, radius=1.5246, center=<0.2363 0.0535 0.3043>, rotation=<rot 90 0 0>}
  {shapeType=Capsule, length=16.8405, radius=1.2523, center=<-8.7668 -3.1523 2.4332>, rotation=<rot 41.6827 -0.8363 8.894>}
  {shapeType=Capsule, length=4.7519, radius=0.5, center=<-14.7896 -2.087 7.5829>, rotation=<rot 38.4373 -1.924 6.9419>}
  {shapeType=Capsule, length=15.2107, radius=4.1907, center=<0.4863 -5.8833 -1.1923>, rotation=<rot 36.9888 0 0>}
 ]
}
constraints [
 {bone1Name=dyn_teeth_01, bone2Name=dyn_teeth_MSTR, swing1LimitDegrees=15, swing2LimitDegrees=10, twistLimitDegrees=1}
 {bone1Name=dyn_teeth_02, bone2Name=dyn_teeth_MSTR, swing1LimitDegrees=15, swing2LimitDegrees=10, twistLimitDegrees=1}
 {bone1Name=dyn_teeth_03, bone2Name=dyn_teeth_MSTR, swing1LimitDegrees=15, swing2LimitDegrees=10, twistLimitDegrees=1}
 {bone1Name=dyn_teeth_04, bone2Name=dyn_teeth_MSTR, swing1LimitDegrees=15, swing2LimitDegrees=10, twistLimitDegrees=1}
 {bone1Name=dyn_teeth_05, bone2Name=dyn_teeth_MSTR, swing1LimitDegrees=15, swing2LimitDegrees=10, twistLimitDegrees=1}
 {bone1Name=dyn_teeth_06, bone2Name=dyn_teeth_MSTR, swing1LimitDegrees=15, swing2LimitDegrees=10, twistLimitDegrees=1}
 {bone1Name=dyn_teeth_07, bone2Name=dyn_teeth_MSTR, swing1LimitDegrees=15, swing2LimitDegrees=10, twistLimitDegrees=1}
 {bone1Name=dyn_rope_MSTR, bone2Name=pack, swing1LimitDegrees=45, swing2LimitDegrees=45, twistLimitDegrees=15}
 {bone1Name=dyn_rope_l_01, bone2Name=dyn_rope_MSTR, swing1LimitDegrees=65, swing2LimitDegrees=65, twistLimitDegrees=1}
 {bone1Name=dyn_rope_l_02, bone2Name=dyn_rope_l_01, swing1LimitDegrees=45, swing2LimitDegrees=45, twistLimitDegrees=1}
 {bone1Name=dyn_rope_l_03, bone2Name=dyn_rope_l_02, swing1LimitDegrees=45, swing2LimitDegrees=45, twistLimitDegrees=1}
 {bone1Name=dyn_rope_r_01, bone2Name=dyn_rope_MSTR, swing1LimitDegrees=55, swing2LimitDegrees=55, twistLimitDegrees=1}
 {bone1Name=dyn_rope_r_02, bone2Name=dyn_rope_r_01, swing1LimitDegrees=45, swing2LimitDegrees=45, twistLimitDegrees=1}
 {bone1Name=dyn_rope_r_03, bone2Name=dyn_rope_r_02, swing1LimitDegrees=45, swing2LimitDegrees=45, twistLimitDegrees=1}
 {
  bone1Name=dyn_teeth_MSTR
  bone2Name=pack
  swing1LimitDegrees=45
  swing1Motion=Free
  swing2LimitDegrees=45
  swing2Motion=Free
  twistLimitDegrees=45
  twistMotion=Free
 }
 {bone1Name=dyn_top_strap, bone2Name=pack, swing1LimitDegrees=45, swing2LimitDegrees=12, twistLimitDegrees=1}
 {bone1Name=dyn_belt_buckle, bone2Name=pack, swing1LimitDegrees=3, swing2LimitDegrees=3, twistLimitDegrees=1}
 {bone1Name=dyn_baseball_MSTR, bone2Name=pack, swing1LimitDegrees=1, swing2LimitDegrees=50, twistLimitDegrees=1}
 {bone1Name=dyn_air_hose_01, bone2Name=pack, swing1LimitDegrees=15, swing2LimitDegrees=45, twistLimitDegrees=1}
 {bone1Name=dyn_air_hose_02, bone2Name=pack, swing1LimitDegrees=15, swing2LimitDegrees=10, twistLimitDegrees=3}
 {bone1Name=dyn_air_hose_03, bone2Name=dyn_air_hose_02, swing1LimitDegrees=15, swing2LimitDegrees=15, twistLimitDegrees=1}
 {bone1Name=dyn_seashell, bone2Name=pack, swing1LimitDegrees=25, swing2LimitDegrees=70, twistLimitDegrees=1}
 {bone1Name=dyn_bottlecap, bone2Name=pack, swing1LimitDegrees=35, swing2LimitDegrees=90, twistLimitDegrees=1}
]
